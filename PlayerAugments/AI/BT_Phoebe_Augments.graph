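# Behavior tree for Phoebe AI bots that handles Player Augments, using the BB_Augment blackboard.
# A root Selector holds a single FortAthenaBTTask_Augments task with no decorators or services.
# The task is bound to the evaluator's augment item and execution status keys, with a minimum and maximum selection time bounding how long a selection takes.
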BehaviorTree /PlayerAugments/AI/BT_Phoebe_Augments
blackboard=&/BB_Augment
root {
 name=Selector
 class=BTComposite_Selector
 children [Augments {
   class=FortAthenaBTTask_Augments
   children []
   decorators []
   properties {
    augmentItemKeyName=AIEvaluator_Augments_Item
    augmentStatusKeyName=AIEvaluator_Augments_ExecutionStatus
    bIgnoreRestartSelf=false
    maxSelectionTime=5
    minSelectionTime=2
    nodeName=Augments
   }
   services []
  }]
 decorators []
 properties {children [{childComposite=null, childTask=&self:FortAthenaBTTask_Augments_0, decorators []}], services []}
 services []
}
